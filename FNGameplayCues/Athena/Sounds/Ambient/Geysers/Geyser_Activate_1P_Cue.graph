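# Sound cue for the first-person sound of a geyser activation launch in Athena.
# A random node picks one of three equally weighted launch wave variations.
# An attenuation override turns off distance attenuation and spatialization, so it plays as non-positional audio with boosted volume in the Athena ambient sound class.
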SoundCue /FNGameplayCues/Athena/Sounds/Ambient/Geysers/Geyser_Activate_1P_Cue
root=Random_0
nodes [
 {id=Random_0, children [WavePlayer_0,WavePlayer_1,WavePlayer_2], properties {weights [1,1,1]}}
 {id=WavePlayer_0, properties {soundWaveAssetPtr=/Game/Athena/Sounds/Ambient/Geysers/Geyser_Launch_1P_02}}
 {id=WavePlayer_1, properties {soundWaveAssetPtr=/Game/Athena/Sounds/Ambient/Geysers/Geyser_Launch_1P_03}}
 {id=WavePlayer_2, properties {soundWaveAssetPtr=/Game/Athena/Sounds/Ambient/Geysers/Geyser_Launch_1P_01}}
]
properties {
 bOverrideAttenuation=true
 maxDistance=2097152
 soundClassObject=&Fort_Ambient_Athena
 volumeMultiplier=2.079
 attenuationOverrides {bAttenuate=false, bSpatialize=false}
}
